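; Stealth grass building actor that applies a stealth GameplayEffect to overlapping player pawns and removes it, plus the crouch variant, on exit.
; BeginPlay registers a collision response swapper for AI visibility, or self-destructs when disabled.
; Only damage with allowed tags gets through.
; Debug events rebuild the HISM by tracing a jittered grid, using noise-material luminance read from a render target to reject and optionally scale seeded-random instances.
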(event EventBeginPlay
  ;[1 1 1] Required for AIs to handle visibility of player going in/out of grass
  (Components|InstancedStaticMesh|SetCullDistances (Variables|Default|GetGrassHISM) 28000 28000)
  (LOD|SetMaxDrawDistance (Variables|Default|GetGrassHISM) 2.8e+04)
  (if (ScalableFloat|GetValueasBool (Variables|Default|GetGrassEnabledRow))
    (Class|FortCollisionResponseSwapperComponent|RegisterCallbacksforOverlap (Variables|Default|GetFortCollisionResponseSwapper) (Variables|Default|GetCylinderOverlapVolume) (Variables|Default|GetCylinderOverlapVolume))
    (else
      (Actor|SetLifeSpan (Math|Random|RandomFloatinRange 0.1 5.0)) ; Self destruct
    )))

(event EventEndPlay
  (params (EndPlayReason :type EEndPlayReason%20Enum :default "Destroyed"))
  ;[1 1 1] Required for AIs to handle visibility of player going in/out of grass
  (Class|FortCollisionResponseSwapperComponent|UnregisterCallbacksforOverlap (Variables|Default|GetFortCollisionResponseSwapper)))

(event OnComponentBeginOverlap(CylinderOverlapVolume)
  (params (OverlappedComponent :type Primitive%20Component%20Object%20Reference) (OtherActor :type Actor%20Object%20Reference) (OtherComp :type Primitive%20Component%20Object%20Reference) (OtherBodyIndex :type Integer) (bFromSweep :type Boolean) (SweepResult :type Hit%20Result%20Structure%20%28by%20ref%29))
  ;[1 1 1] Whenever you overlap a grass volume, you get this GE unless you already have it
  (bind _asplayer_pawn_athena (Utilities|Casting|CastToPlayerPawn_Athena OtherActor))
  (if true
    (Networking|FlushNetDormancy)
    (else
      _))
  (Utilities|IsValid (Ability|GetAbilitySystemComponent _asplayer_pawn_athena)
    (:Is%20Valid
      (GameplayEffects|ApplyGameplayEffectToTarget :self (Ability|GetAbilitySystemComponent _asplayer_pawn_athena) :GameplayEffectClass (Variables|Default|GetGE_StealthGrass) :Target (Ability|GetAbilitySystemComponent _asplayer_pawn_athena) :Level 1.0)
      (Networking|ForceNetUpdate _asplayer_pawn_athena))))

(event OnComponentEndOverlap(CylinderOverlapVolume)
  (params (OverlappedComponent :type Primitive%20Component%20Object%20Reference) (OtherActor :type Actor%20Object%20Reference) (OtherComp :type Primitive%20Component%20Object%20Reference) (OtherBodyIndex :type Integer))
  ;[1 1 1] Whenever you stop overlapping the grass volume, remove the GE
  (bind _asplayer_pawn_athena (Utilities|Casting|CastToPlayerPawn_Athena OtherActor))
  (GameplayEffects|RemoveActiveGameplayEffectbySourceEffect (Ability|GetAbilitySystemComponent _asplayer_pawn_athena) (Variables|Default|GetGE_StealthGrass) 0 1)
  (GameplayEffects|RemoveActiveGameplayEffectbySourceEffect (Ability|GetAbilitySystemComponent _asplayer_pawn_athena) (Variables|Default|GetGE_StealthGrassCrouch) 0 1)
  (Networking|ForceNetUpdate _asplayer_pawn_athena))

(event Custom|CallPlaceGrass
  ;[1 1 1] Debug Utilities
  (CallFunction|PlaceGrass))

(event Custom|ClearDebugLines
  ;[1 1 1] Debug Utilities
  (Rendering|Debug|FlushPersistentDebugLines))

(event Custom|ClearISMs
  ;[1 1 1] Debug Utilities
  (Components|InstancedStaticMesh|ClearInstances (Variables|Default|GetGrassHISM)))

(event Custom|TogglePreviewPlaneVisibility
  (bind _noisepreviewplane (Variables|Default|GetNoisePreviewPlane))
  (bind _bvisible (Variables|/Script/Engine.SceneComponent|GetbVisible _noisepreviewplane))
  ;[1 1 1] Debug Utilities
  (Development|SetHiddeninGame _noisepreviewplane _bvisible false)
  (Rendering|SetVisibility _noisepreviewplane (not _bvisible) false))

(event Custom|UpdateAllGrassActorsInLevel
  ;[1 1 1] Debug Utilities
  (bind _outactors (Actor|GetAllActorsOfClass "/Game/Athena/Artemis/Environments/BuildingActors/Foliage/B_Artemis_StealthGrass.B_Artemis_StealthGrass_C"))
  (for _array_element _outactors
    (Utilities|IsValid _array_element
      (:Is%20Valid
        (CallFunction|CallPlaceGrass _array_element)))))

(fn BlueprintModifyIncomingDamage
  (params (Damage :type Float%20%28single-precision%29) (InTags :type Gameplay%20Tag%20Container%20Structure%20%28by%20ref%29) (EffectContext :type Gameplay%20Effect%20Context%20Handle%20Structure%20%28by%20ref%29) (EventInstigator :type Controller%20Object%20Reference) (DamageCauser :type Actor%20Object%20Reference))
  (if (GameplayTags|HasAnyTags InTags (Variables|Default|GetAllowedDamageTags) true)
    (bind _returnvalue (|CallParentFunction Damage InTags EffectContext EventInstigator DamageCauser))
    (return _returnvalue)
    (else
      (return 0.0))))

(fn CalculateTraceOriginWorldSpace
  (params (LoopIndex :type Integer) (TraceVolumeBoxExtentScaledX :type Float%20%28double-precision%29) (TraceVolumeBoxExtent :type Vector))
  (bind (_remainder _returnvalue) (Math|Float|Division(WholeandRemainder) (Math|Conversions|ToFloat(Integer) LoopIndex) (Math|Conversions|ToFloat(Integer) (Math|Float|Floor TraceVolumeBoxExtentScaledX))))
  (bind (_tracevolumeboxextent_x _tracevolumeboxextent_y _tracevolumeboxextent_z) (Math|Vector|BreakVector (Variables|GetTraceVolumeBoxExtent_X)))
  (bind (_placementjitter_x _placementjitter_y) (Math|Vector2D|BreakVector2D (Variables|GetPlacementJitter_X)))
  (return (Math|Transform|TransformLocation (Transformation|GetActorTransform) (Math|Vector|MakeVector (+ (+ (- (* _remainder (Variables|GrassSettings|GetGrassDensityDivisor)) _tracevolumeboxextent_x) (/ (Variables|GrassSettings|GetGrassDensityDivisor) 2)) (* (Math|Random|RandomFloatinRangefromStream (Math|Random|MakeRandomStream (+ (Variables|GrassAdvancedSettings|GetRandomSeed) (Variables|GetLoopIndex))) -1.0 1.0) _placementjitter_x)) (+ (+ (/ (Variables|GrassSettings|GetGrassDensityDivisor) 2) (- (* _returnvalue (Variables|GrassSettings|GetGrassDensityDivisor)) _tracevolumeboxextent_y)) (* _placementjitter_y (Math|Random|RandomFloatinRangefromStream (Math|Random|MakeRandomStream (+ (Variables|GrassAdvancedSettings|GetRandomSeed) (Variables|GetLoopIndex))) -1.0 1.0))) _tracevolumeboxextent_z))))

(fn GetCurieCollisionComponentOverride
  (return (Variables|Default|GetCylinderOverlapVolume)))

(fn InitializePreviewPlane
  (bind (_returnvalue_x _returnvalue_y _returnvalue_z) (Math|Vector|BreakVector (Components|Box|GetScaledBoxExtent (Variables|Default|GetTraceVolume))))
  (Transformation|SetWorldScale3D (Variables|Default|GetNoisePreviewPlane) (* (Components|Box|GetScaledBoxExtent (Variables|Default|GetTraceVolume)) "2.000000,2.000000,2.000000"))
  (Transformation|SetRelativeLocation (Variables|Default|GetNoisePreviewPlane) (Math|Vector|MakeVector 0.0 0.0 _returnvalue_z) false false)
  (bind _returnvalue (Rendering|Material|CreateDynamicMaterialInstance (Variables|GrassSettings|GetProjectionNoiseMaterial) "None" "None" :!class /Script/Engine.KismetMaterialLibrary))
  (bind _output_get (Variables|Default|SetPreviewPlaneProjectionNoiseMID _returnvalue))
  (Rendering|Material|SetMaterial (Variables|Default|GetNoisePreviewPlane) 0 _output_get))

(fn InitializeRenderTarget
  (bind (_min _max) (Components|StaticMesh|GetLocalBounds (Variables|Default|GetNoisePreviewPlane)))
  (bind (_rendertargetresolution_x _rendertargetresolution_y) (Math|Vector2D|BreakVector2D (Variables|GetRenderTargetResolution_X)))
  (bind _returnvalue (CanvasRenderTarget2D|CreateCanvasRenderTarget2D "/Script/Engine.CanvasRenderTarget2D" (Math|Float|Truncate _rendertargetresolution_x) (Math|Float|Truncate _rendertargetresolution_y)))
  (Variables|Default|SetColorLookupRT _returnvalue)
  (bind _returnvalue_2 (Rendering|Material|CreateDynamicMaterialInstance (Variables|GrassSettings|GetProjectionNoiseMaterial) "None" "None" :!class /Script/Engine.KismetMaterialLibrary))
  (bind _output_get (Variables|Default|SetRTProjectionNoiseMID _returnvalue_2))
  (Rendering|Material|SetScalarParameterValue _output_get "isRenderTarget" 1.0 :!class /Script/Engine.MaterialInstanceDynamic)
  (Rendering|Material|SetScalarParameterValue (Variables|Default|GetRTProjectionNoiseMID) "WorldPositionDivisor" (Variables|GrassSettings|GetTextureNoiseScale) :!class /Script/Engine.MaterialInstanceDynamic)
  (Rendering|Material|SetScalarParameterValue (Variables|Default|GetRTProjectionNoiseMID) "SphereMaskRadius" (* (Math|Float|Max(Float) (.x (* _max (Variables|/Script/Engine.SceneComponent|GetRelativeScale3D (Variables|Default|GetNoisePreviewPlane)))) (.y (* _max (Variables|/Script/Engine.SceneComponent|GetRelativeScale3D (Variables|Default|GetNoisePreviewPlane)))) (.z (* _max (Variables|/Script/Engine.SceneComponent|GetRelativeScale3D (Variables|Default|GetNoisePreviewPlane))))) (Variables|GrassSettings|GetSphericalMaskRadius)) :!class /Script/Engine.MaterialInstanceDynamic)
  (Rendering|Material|SetScalarParameterValue (Variables|Default|GetRTProjectionNoiseMID) "SphereMaskHardness" (Variables|GrassSettings|GetSphericalMaskHardness) :!class /Script/Engine.MaterialInstanceDynamic)
  (Rendering|Material|SetVectorParameterValue (Variables|Default|GetPreviewPlaneProjectionNoiseMID) "PreviewPlanePosition" (Math|Conversions|ToLinearColor(Vector) (Transformation|GetWorldLocation (Variables|Default|GetNoisePreviewPlane))) :!class /Script/Engine.MaterialInstanceDynamic)
  (Rendering|Material|SetScalarParameterValue (Variables|Default|GetRTProjectionNoiseMID) "NoiseExponent" (Variables|GrassSettings|GetNoiseContrast) :!class /Script/Engine.MaterialInstanceDynamic)
  (Rendering|Material|SetScalarParameterValue (Variables|Default|GetRTProjectionNoiseMID) "NoiseBias" (Variables|GrassSettings|GetNoiseBias) :!class /Script/Engine.MaterialInstanceDynamic)
  (Rendering|Material|SetScalarParameterValue (Variables|Default|GetRTProjectionNoiseMID) "FinalScale" (Variables|GrassSettings|GetFinalScale) :!class /Script/Engine.MaterialInstanceDynamic))

(fn InitializeVolumes
  (Components|Box|SetBoxExtent (Variables|Default|GetTraceVolume) (Variables|GrassSettings|GetVolumeBoxExtent) true)
  (Transformation|SetWorldScale3D (Variables|Default|GetCylinderOverlapVolume) (* (select (Variables|GrassAdvancedSettings|GetSphericalMaskRadiusModulatesOverlapVolumeScale) (* (/ (Variables|GrassSettings|GetVolumeBoxExtent) 5e+01) (Variables|GrassSettings|GetSphericalMaskRadius)) (/ (Variables|GrassSettings|GetVolumeBoxExtent) 5e+01)) (Variables|GrassAdvancedSettings|GetOverlapVolumeBoxExtentScale))))

(fn LookupRenderTargetColor
  (params (WorldPosition :type Vector))
  (bind _rtprojectionnoisemid (Variables|Default|GetRTProjectionNoiseMID))
  (Rendering|Material|SetVectorParameterValue _rtprojectionnoisemid "WorldPosition" (Math|Conversions|ToLinearColor(Vector) WorldPosition) :!class /Script/Engine.MaterialInstanceDynamic)
  (Rendering|DrawMaterialtoRenderTarget (Variables|Default|GetColorLookupRT) _rtprojectionnoisemid)
  (bind _returnvalue (Rendering|ReadRenderTargetRawUV (Variables|Default|GetColorLookupRT) 0.5 0.5))
  (return _returnvalue))

(fn PerformTracesAndPlaceISMs
  (bind (_xaxis_1_1118a77842db438295cc7fb043bdddf9 _yaxis_3_2dec3edf470d60a1ea4cf1a0d55cba4c _zaxis_5_d51bdfca433b784f2ea235a705a0a705) (Utilities|Struct|BreakBArtemisStealthGrassScaledAxes (Variables|GrassSettings|GetScaledAxesByLuminance)))
  (bind _tracevolume (Variables|Default|GetTraceVolume))
  (bind _returnvalue (Components|Box|GetScaledBoxExtent _tracevolume))
  (bind _grassdensitydivisor (Variables|GrassSettings|GetGrassDensityDivisor))
  (bind _returnvalue_2 (* _returnvalue 2.0))
  (bind _returnvalue_3 (/ _returnvalue_2 _grassdensitydivisor))
  (bind (_returnvalue_x _returnvalue_y) (Utilities|Struct|BreakIntVector (Math|Float|Truncate(Vector) _returnvalue_3)))
  (bind (_returnvalue_x_2 _returnvalue_y_2 _returnvalue_z) (Math|Vector|BreakVector (* (LocalVariables|GetRenderTargetLuminance) (Variables|GrassSettings|GetGrassScale))))
  (bind (_grassscale_x _grassscale_y _grassscale_z) (Math|Vector|BreakVector (Variables|GetGrassScale_X)))
  (bind _tracevolumeboxextentscaled_x (Math|Vector|BreakVector (Variables|GetTraceVolumeBoxExtentScaled_X)))
  (CallFunction|ClearDebugLines)
  (LocalVariables|SetTraceVolumeBoxExtent _returnvalue)
  (LocalVariables|SetTraceVolumeBoxBounds _returnvalue_2)
  (LocalVariables|SetTraceVolumeBoxExtentScaled _returnvalue_3)
  (for _index (range (* _returnvalue_x _returnvalue_y))
    (bind _output_get (LocalVariables|SetLoopIndex _index))
    (bind _output_get_2 (LocalVariables|SetTraceLocationOrigin (CallFunction|CalculateTraceOriginWorldSpace _output_get _tracevolumeboxextentscaled_x (LocalVariables|GetTraceVolumeBoxExtent))))
    (if (Variables|GrassAdvancedSettings|GetDrawTraceOriginDebugSpheres)
      (Rendering|Debug|DrawDebugSphere _output_get_2 (Variables|GrassAdvancedSettings|GetDebugSpheresRadius) 4 "(R=1.000000,G=0.000000,B=0.945354,A=1.000000)" 1e+03 1.0)
      (else
        _))
    (bind _linearcolor (CallFunction|LookupRenderTargetColor (LocalVariables|GetTraceLocationOrigin)))
    (bind _output_get_3 (LocalVariables|SetRenderTargetColor _linearcolor))
    (LocalVariables|SetRenderTargetLuminance (Math|Color|Luminance(LinearColor) _output_get_3))
    (if (>= (Math|Color|Luminance(LinearColor) _output_get_3) (Variables|GrassAdvancedSettings|GetRejectionSamplingThreshold))
      (bind (_outhit _returnvalue_4) (Collision|LineTraceByChannel (LocalVariables|GetTraceLocationOrigin) (+ (LocalVariables|GetTraceLocationOrigin) (Math|Transform|TransformDirection (Transformation|GetActorTransform) (Variables|GrassSettings|GetTraceDirectionalVector))) "TraceTypeQuery10" true (Variables|GrassSettings|GetTraceActorsToIgnore) "ForDuration" true (LocalVariables|GetRenderTargetColor) "(R=0.000000,G=1.000000,B=0.000000,A=1.000000)" (Variables|GrassAdvancedSettings|GetDebugTraceDrawTime)))
      (if _returnvalue_4
        (if (not (Variables|GrassAdvancedSettings|GetDebugSkipLandscapeCasts))
          (Utilities|Casting|CastToLandscapeStreamingProxy (.HitActor (Collision|BreakHitResult _outhit))
            (:then
              _)
            (:CastFailed
              (Utilities|Casting|CastToLandscape (.HitActor (Collision|BreakHitResult _outhit)))))
          (if (<= (* (+ (Math|Vector|DotProduct (.Normal (Collision|BreakHitResult _outhit)) "0, 0,-1.000000") 1.0) 1.8e+02) (Variables|GrassAdvancedSettings|GetMaxAngleToSpawn))
            (Components|InstancedStaticMesh|AddInstance (Variables|Default|GetGrassHISM) (Math|Transform|MakeTransform (.Location (Collision|BreakHitResult _outhit)) (Math|Rotator|CombineRotators (Math|Rotator|MakeRotfromZ (.Normal (Collision|BreakHitResult _outhit))) (Math|Vector|RotatorfromAxisandAngle (.Normal (Collision|BreakHitResult _outhit)) (Math|Random|RandomFloatinRangefromStream (Math|Random|MakeRandomStream (+ (Variables|GrassAdvancedSettings|GetRandomSeed) (LocalVariables|GetLoopIndex))) 0.0 3.6e+02))) (+ (Math|Vector|MakeVector (select _xaxis_1_1118a77842db438295cc7fb043bdddf9 _returnvalue_x_2 _grassscale_x) (select _yaxis_3_2dec3edf470d60a1ea4cf1a0d55cba4c _returnvalue_y_2 _grassscale_y) (select _zaxis_5_d51bdfca433b784f2ea235a705a0a705 _returnvalue_z _grassscale_z)) (* (Math|Random|RandomFloatinRangefromStream (Math|Random|MakeRandomStream (+ (Variables|GrassAdvancedSettings|GetRandomSeed) (LocalVariables|GetLoopIndex))) (* (Variables|GrassSettings|GetGrassScaleRandomness%25) -1.0) (Variables|GrassSettings|GetGrassScaleRandomness%25)) (Variables|GrassSettings|GetGrassScale)))) true)))))))

(fn PlaceGrass
  (Components|InstancedStaticMesh|ClearInstances (Variables|Default|GetGrassHISM))
  (CallFunction|SetGrassMaterial)
  (CallFunction|InitializeVolumes)
  (CallFunction|InitializePreviewPlane)
  (CallFunction|InitializeRenderTarget)
  (CallFunction|PerformTracesandPlaceISMs))

(fn SetGrassMaterial
  (Rendering|Material|SetMaterial (Variables|Default|GetGrassHISM) 0 (select (or (TouchPlatform|IsMobileGame) (TouchPlatform|IsNintendoSwitch)) (Variables|GrassAdvancedSettings|GetGrassMaterialMobile) (Variables|GrassAdvancedSettings|GetGrassMaterial))))

(fn ConstructionScript
  (if (Editor|InEditor)
    (CallFunction|InitializePreviewPlane)
    (CallFunction|InitializeRenderTarget)))
