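; Creative stealth grass actor that holds instanced grass in a box volume and sets its cull and draw distances at BeginPlay for AI visibility handling.
; PlaceGrass clears instances, picks mobile/Switch or default material, and configures a preview plane and noise render target.
; Helpers compute seeded jittered grid origins and sample noise per world position by render-target readback.
; The trace-and-place step only clears debug lines; debug events refresh or clear grass.
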(event EventBeginPlay
  ;[1 1 1] Required for AIs to handle visibility of player going in/out of grass
  (Components|InstancedStaticMesh|SetCullDistances (Variables|Default|GetGrassHISM) 28000 28000)
  (LOD|SetMaxDrawDistance (Variables|Default|GetGrassHISM) 2.8e+04))

(event Custom|CallPlaceGrass
  ;[1 1 1] Debug Utilities
  (CallFunction|PlaceGrass))

(event Custom|ClearDebugLines
  ;[1 1 1] Debug Utilities
  (Rendering|Debug|FlushPersistentDebugLines))

(event Custom|ClearISMs
  ;[1 1 1] Debug Utilities
  (Components|InstancedStaticMesh|ClearInstances (Variables|Default|GetGrassHISM)))

(event Custom|TogglePreviewPlaneVisibility
  (bind _noisepreviewplane (Variables|Default|GetNoisePreviewPlane))
  (bind _bvisible (Variables|/Script/Engine.SceneComponent|GetbVisible _noisepreviewplane))
  ;[1 1 1] Debug Utilities
  (Development|SetHiddeninGame _noisepreviewplane _bvisible false)
  (Rendering|SetVisibility _noisepreviewplane (not _bvisible) false))

(event Custom|UpdateAllGrassActorsInLevel
  ;[1 1 1] Debug Utilities
  (bind _outactors (Actor|GetAllActorsOfClass "/Game/Athena/Artemis/Environments/BuildingActors/Foliage/B_Artemis_StealthGrass_Creative.B_Artemis_StealthGrass_Creative_C"))
  (for _array_element _outactors
    (Utilities|IsValid _array_element
      (:Is%20Valid
        (CallFunction|CallPlaceGrass _array_element)))))

(fn CalculateTraceOriginWorldSpace
  (params (LoopIndex :type Integer) (TraceVolumeBoxExtentScaledX :type Float%20%28double-precision%29) (TraceVolumeBoxExtent :type Vector))
  (bind (_remainder _returnvalue) (Math|Float|Division(WholeandRemainder) (Math|Conversions|ToFloat(Integer) LoopIndex) (Math|Conversions|ToFloat(Integer) (Math|Float|Floor TraceVolumeBoxExtentScaledX))))
  (bind (_tracevolumeboxextent_x _tracevolumeboxextent_y _tracevolumeboxextent_z) (Math|Vector|BreakVector (Variables|GetTraceVolumeBoxExtent_X)))
  (bind (_placementjitter_x _placementjitter_y) (Math|Vector2D|BreakVector2D (Variables|GetPlacementJitter_X)))
  (return (Math|Transform|TransformLocation (Transformation|GetActorTransform) (Math|Vector|MakeVector (+ (+ (- (* _remainder (Variables|GrassSettings|GetGrassDensityDivisor)) _tracevolumeboxextent_x) (/ (Variables|GrassSettings|GetGrassDensityDivisor) 2)) (* (Math|Random|RandomFloatinRangefromStream (Math|Random|MakeRandomStream (+ (Variables|GrassAdvancedSettings|GetRandomSeed) (Variables|GetLoopIndex))) -1.0 1.0) _placementjitter_x)) (+ (+ (/ (Variables|GrassSettings|GetGrassDensityDivisor) 2) (- (* _returnvalue (Variables|GrassSettings|GetGrassDensityDivisor)) _tracevolumeboxextent_y)) (* _placementjitter_y (Math|Random|RandomFloatinRangefromStream (Math|Random|MakeRandomStream (+ (Variables|GrassAdvancedSettings|GetRandomSeed) (Variables|GetLoopIndex))) -1.0 1.0))) _tracevolumeboxextent_z))))

(fn InitializePreviewPlane
  (bind (_returnvalue_x _returnvalue_y _returnvalue_z) (Math|Vector|BreakVector (Components|Box|GetScaledBoxExtent (Variables|Default|GetTraceVolume))))
  (Transformation|SetWorldScale3D (Variables|Default|GetNoisePreviewPlane) (* (Components|Box|GetScaledBoxExtent (Variables|Default|GetTraceVolume)) "2.000000,2.000000,2.000000"))
  (Transformation|SetRelativeLocation (Variables|Default|GetNoisePreviewPlane) (Math|Vector|MakeVector 0.0 0.0 _returnvalue_z) false false)
  (bind _returnvalue (Rendering|Material|CreateDynamicMaterialInstance (Variables|GrassSettings|GetProjectionNoiseMaterial) "None" "None" :!class /Script/Engine.KismetMaterialLibrary))
  (bind _output_get (Variables|Default|SetPreviewPlaneProjectionNoiseMID _returnvalue))
  (Rendering|Material|SetMaterial (Variables|Default|GetNoisePreviewPlane) 0 _output_get))

(fn InitializeRenderTarget
  (bind (_min _max) (Components|StaticMesh|GetLocalBounds (Variables|Default|GetNoisePreviewPlane)))
  (bind (_rendertargetresolution_x _rendertargetresolution_y) (Math|Vector2D|BreakVector2D (Variables|GetRenderTargetResolution_X)))
  (bind _returnvalue (CanvasRenderTarget2D|CreateCanvasRenderTarget2D "/Script/Engine.CanvasRenderTarget2D" (Math|Float|Truncate _rendertargetresolution_x) (Math|Float|Truncate _rendertargetresolution_y)))
  (Variables|Default|SetColorLookupRT _returnvalue)
  (bind _returnvalue_2 (Rendering|Material|CreateDynamicMaterialInstance (Variables|GrassSettings|GetProjectionNoiseMaterial) "None" "None" :!class /Script/Engine.KismetMaterialLibrary))
  (bind _output_get (Variables|Default|SetRTProjectionNoiseMID _returnvalue_2))
  (Rendering|Material|SetScalarParameterValue _output_get "isRenderTarget" 1.0 :!class /Script/Engine.MaterialInstanceDynamic)
  (Rendering|Material|SetScalarParameterValue (Variables|Default|GetRTProjectionNoiseMID) "WorldPositionDivisor" (Variables|GrassSettings|GetTextureNoiseScale) :!class /Script/Engine.MaterialInstanceDynamic)
  (Rendering|Material|SetScalarParameterValue (Variables|Default|GetRTProjectionNoiseMID) "SphereMaskRadius" (* (Math|Float|Max(Float) (.x (* _max (Variables|/Script/Engine.SceneComponent|GetRelativeScale3D (Variables|Default|GetNoisePreviewPlane)))) (.y (* _max (Variables|/Script/Engine.SceneComponent|GetRelativeScale3D (Variables|Default|GetNoisePreviewPlane)))) (.z (* _max (Variables|/Script/Engine.SceneComponent|GetRelativeScale3D (Variables|Default|GetNoisePreviewPlane))))) (Variables|GrassSettings|GetSphericalMaskRadius)) :!class /Script/Engine.MaterialInstanceDynamic)
  (Rendering|Material|SetScalarParameterValue (Variables|Default|GetRTProjectionNoiseMID) "SphereMaskHardness" (Variables|GrassSettings|GetSphericalMaskHardness) :!class /Script/Engine.MaterialInstanceDynamic)
  (Rendering|Material|SetVectorParameterValue (Variables|Default|GetPreviewPlaneProjectionNoiseMID) "PreviewPlanePosition" (Math|Conversions|ToLinearColor(Vector) (Transformation|GetWorldLocation (Variables|Default|GetNoisePreviewPlane))) :!class /Script/Engine.MaterialInstanceDynamic)
  (Rendering|Material|SetScalarParameterValue (Variables|Default|GetRTProjectionNoiseMID) "NoiseExponent" (Variables|GrassSettings|GetNoiseContrast) :!class /Script/Engine.MaterialInstanceDynamic)
  (Rendering|Material|SetScalarParameterValue (Variables|Default|GetRTProjectionNoiseMID) "NoiseBias" (Variables|GrassSettings|GetNoiseBias) :!class /Script/Engine.MaterialInstanceDynamic)
  (Rendering|Material|SetScalarParameterValue (Variables|Default|GetRTProjectionNoiseMID) "FinalScale" (Variables|GrassSettings|GetFinalScale) :!class /Script/Engine.MaterialInstanceDynamic))

(fn InitializeVolumes)

(fn LookupRenderTargetColor
  (params (WorldPosition :type Vector))
  (bind _rtprojectionnoisemid (Variables|Default|GetRTProjectionNoiseMID))
  (Rendering|Material|SetVectorParameterValue _rtprojectionnoisemid "WorldPosition" (Math|Conversions|ToLinearColor(Vector) WorldPosition) :!class /Script/Engine.MaterialInstanceDynamic)
  (Rendering|DrawMaterialtoRenderTarget (Variables|Default|GetColorLookupRT) _rtprojectionnoisemid)
  (bind _returnvalue (Rendering|ReadRenderTargetRawUV (Variables|Default|GetColorLookupRT) 0.5 0.5))
  (return _returnvalue))

(fn PerformTracesAndPlaceISMs
  (CallFunction|ClearDebugLines))

(fn PlaceGrass
  (Components|InstancedStaticMesh|ClearInstances (Variables|Default|GetGrassHISM))
  (CallFunction|SetGrassMaterial)
  (CallFunction|InitializeVolumes)
  (CallFunction|InitializePreviewPlane)
  (CallFunction|InitializeRenderTarget)
  (CallFunction|PerformTracesandPlaceISMs))

(fn SetGrassMaterial
  (Rendering|Material|SetMaterial (Variables|Default|GetGrassHISM) 0 (select (or (TouchPlatform|IsMobileGame) (TouchPlatform|IsNintendoSwitch)) (Variables|GrassAdvancedSettings|GetGrassMaterialMobile) (Variables|GrassAdvancedSettings|GetGrassMaterial))))

(fn ConstructionScript)
