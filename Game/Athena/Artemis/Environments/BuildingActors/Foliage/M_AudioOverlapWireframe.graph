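# M_AudioOverlapWireframe is an unlit wireframe material, apparently for visualizing audio overlap shapes, judging by its name, and it sits with the Artemis foliage building actors.
# Its graph holds a single constant green color node, with no textures or parameters.
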Material /Game/Athena/Artemis/Environments/BuildingActors/Foliage/M_AudioOverlapWireframe
graph {
 properties {pixelDepthOffsetMode=PDOM_Legacy, shadingModel=MSM_Unlit, usedShadingModels=Unlit, wireframe=true}
 nodes [Constant3Vector() {constant=<rgba 0 1 0 0>}]
}
