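# Translucent, unlit, two-sided material that builds a noise map for Artemis stealth grass, apparently serving both a preview plane and a render target.
# World-space noise is shaped by exponent, bias and scale, and combined with a sphere mask centered on the preview plane's XY position.
# A separate branch makes hard animated stripes from a rounded sine of world X.
# The isRenderTarget parameter swaps world position for a vector parameter and blends out the stripes.
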Material /Game/Athena/Artemis/Environments/BuildingActors/Foliage/Materials/M_Artemis_StealthGrass_NoiseMap
graph {
 properties {blendMode=BLEND_Translucent, pixelDepthOffsetMode=PDOM_Legacy, shadingModel=MSM_Unlit, twoSided=true, usedShadingModels=Unlit}
 nodes [
  n1 = Multiply(A=Time(), B=ScalarParameter() {defaultValue=2.5, parameterName=StripesTimeScale})
  n2 = ScalarParameter() {defaultValue=30, parameterName=StripesDivisor}
  n3 = Divide(A=@n2)
  n4 = Divide(A=Add(A=Multiply(A=@n1, B=@n3), B=WorldPosition().XYZ), B=@n2)
  n5 = ConstantBiasScale(None=Sine(None=MaterialFunctionCall(Float3=@n4) {materialFunction=&BreakOutFloat3Components}.R))
  LinearInterpolate(
   A=LinearInterpolate(
    A=ScalarParameter() {defaultValue=0.6, parameterName=StripesLow}
    B=ScalarParameter() {defaultValue=0.8, parameterName=StripesHigh}
    Alpha=Round(None=@n5)
   )
   Alpha=ScalarParameter() {parameterName=isRenderTarget}
  )
  n6 = WorldPosition()
  n7 = ScalarParameter() {defaultValue=512, parameterName=WorldPositionDivisor}
  n8 = VectorParameter() {parameterName=WorldPosition}
  n9 = Noise!(
   World Position=LinearInterpolate(A=Divide(A=@n6.XYZ, B=@n7), B=Divide(A=@n8.RGB, B=@n7), Alpha=ScalarParameter() {parameterName=isRenderTarget})
   FilterWidth
  )
  n10 = ScalarParameter() {defaultValue=1, parameterName=NoiseExponent}
  n11 = Subtract(
   A=LinearInterpolate(A=WorldPosition().XYZ, B=VectorParameter() {parameterName=WorldPosition}.RGB, Alpha=ScalarParameter() {parameterName=isRenderTarget})
   B=VectorParameter() {parameterName=PreviewPlanePosition}.RGB
  )
  n12 = SphereMask(
   A=Abs(None=ComponentMask(None=@n11) {g=true, r=true})
   B=Constant()
   Radius=ScalarParameter() {defaultValue=100, parameterName=SphereMaskRadius}
   Hardness=ScalarParameter() {defaultValue=0.5, parameterName=SphereMaskHardness}
  )
  n13 = Subtract(A=Multiply(A=Power(Base=@n9, Exp=@n10) {constExponent=0.5}, B=@n12), B=ScalarParameter() {defaultValue=0.5, parameterName=NoiseBias}) {constB=0.2}
  Saturate(None=Multiply(A=Add(A=@n13, B=@n12), B=ScalarParameter() {defaultValue=1, parameterName=FinalScale}))
 ]
}
